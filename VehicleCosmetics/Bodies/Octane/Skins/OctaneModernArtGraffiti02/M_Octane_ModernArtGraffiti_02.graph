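# Material for the Modern Art Graffiti 02 skin on the Octane car body, wrapped in the CharacterDeRez_Transition function.
# Base color projects one graffiti texture along two axes from scaled pre-skinned position, composited with a UV-mapped texture and its alpha, then blended by a third texture.
# Switchable, animated detail layers, optionally blended by decal, ramp and noise masks, feed only the normal, with an optional body normal.
# Metallic and roughness are constants.
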Material /VehicleCosmetics/Bodies/Octane/Skins/OctaneModernArtGraffiti02/M_Octane_ModernArtGraffiti_02
graph {
 properties {
  bUseMaterialAttributes=true
  bUsedWithMeshDeformer=true
  bUsedWithSkeletalMesh=true
  pixelDepthOffsetMode=PDOM_Legacy
  usedShadingModels=Default Lit
 }
 nodes [
  Constant2Vector() {desc=TopOffset, g=0.437, r=0.5}
  n1 = VertexInterpolator(VS=Divide(A=PreSkinnedPosition(), B=Constant() {desc=TexScale, r=200}) {desc=PSLP_Scaled}) {desc=PSLP_Scaled_VI}
  n2 = Multiply(
   A=AppendVector(A=ComponentMask(None=@n1.PS) {g=true}, B=ComponentMask(None=@n1.PS) {r=true})
   B=Constant2Vector() {desc=UVTiling, g=0.5, r=0.5}
  )
  n3 = Constant2Vector() {desc=TopOffset, g=0.437, r=0.5}
  n4 = TextureSampleParameter2D(UVs=Frac(None=Add(A=@n2, B=@n3))) {parameterName=Tex_Triplanar, texture=&/Skin_Octane_ModernArtGraffiti_02_BC_2}
  n5 = TextureSample() {texture=&/Skin_Octane_ModernArtGraffiti_02_BC_1}
  n6 = Add(A=Multiply(A=@n4.RGB, B=OneMinus(None=@n5.A)), B=Multiply(A=@n5.RGB, B=@n5.RGB))
  n7 = Multiply(
   A=AppendVector(A=ComponentMask(None=@n1.PS) {b=true}, B=ComponentMask(None=@n1.PS) {r=true})
   B=Constant2Vector() {desc=UVTiling, g=0.5, r=0.5}
  )
  n8 = Multiply(A=Constant() {r=0.5}, B=Constant() {r=5})
  n9 = TextureSampleParameter2D(UVs=Frac(None=Add(A=@n7, B=@n8))) {parameterName=Tex_Triplanar, texture=&Skin_Octane_ModernArtGraffiti_02_BC_2}
  n10 = TextureSampleParameter2D() {parameterName=Tex_Triplanar, texture=&/Skin_Octane_ModernArtGraffiti_02_BC_3}
  n11 = Constant3Vector() {constant=<rgba 0 0 1 0>}
  n12 = Time() {bOverride_Period=true, period=20}
  n13 = VectorParameter() {desc="(R,G) = Detail 1 Pan (X,Y)    (B,A) = Detail 2 Pan (X,Y)", group=Detail, parameterName=DetailTransformPanner}
  n14 = AppendVector(A=@n13.R, B=@n13.R)
  n15 = StaticSwitchParameter(True=TextureCoordinate() {coordinateIndex=1, uTiling=6, vTiling=6}, False=TextureCoordinate() {uTiling=6, vTiling=6}) {defaultValue=true, group=Detail, parameterName=Detail1UseCylinderUV}
  n16 = AppendVector(A=ComponentMask(None=@n15) {g=true}, B=ComponentMask(None=@n15) {r=true})
  n17 = VectorParameter() {desc="(R,G) = Detail1    (B,A) = Detail2", group=Detail, parameterName=DetailTransform, defaultValue=<rgba 1 1 1 1>}
  n18 = Rotator(
   Coordinate=Multiply(
    A=StaticSwitchParameter(True=@n16, False=@n15) {defaultValue=true, group=Detail, parameterName=Detail1SwapXY}
    B=ComponentMask(None=@n17.RGB) {g=true, r=true}
   )
   Time=ScalarParameter() {desc="Rotate Detail1 (0.25 = 90, 0.5 = 180, 0.75 = 270, 1=360)", group=Detail, parameterName=Detail1Rotation}
  ) {speed=6.2832}
  n19 = Time() {bOverride_Period=true, period=20}
  n20 = Multiply(
   A=@n19
   B=ScalarParameter() {desc="Detail1 Animated Spin Speed (Pivot point isn't very predictable)", group=Detail, parameterName=Detail1Spinner}
  )
  n21 = TextureSampleParameter2D(UVs=Add(A=Multiply(A=@n12, B=@n14), B=Rotator(Coordinate=@n18, Time=@n20) {speed=6.2832}) {desc=Panner}) {group=Detail, parameterName=Detail1, samplerSource=SSM_Wrap_WorldGroupSettings, texture=/VehicleCosmetics/Textures/T_Black}
  n22 = LinearInterpolate(A=@n11, B=@n21.RGB, Alpha=Saturate(None=ScalarParameter() {defaultValue=0.25, group=Detail, parameterName=Detail1Strength}))
  n23 = AppendVector(A=@n13.B, B=@n13.B)
  n24 = StaticSwitchParameter(True=TextureCoordinate() {coordinateIndex=1, uTiling=6, vTiling=6}, False=TextureCoordinate() {uTiling=6, vTiling=6}) {defaultValue=true, group=Detail, parameterName=Detail2UseCylinderUV}
  n25 = AppendVector(A=ComponentMask(None=@n24) {g=true}, B=ComponentMask(None=@n24) {r=true})
  n26 = Rotator(
   Coordinate=Multiply(
    A=StaticSwitchParameter(True=@n25, False=@n24) {defaultValue=true, group=Detail, parameterName=Detail2SwapXY}
    B=AppendVector(A=@n17.B, B=@n17.B)
   )
   Time=ScalarParameter() {desc="Rotate Detail2 (0.25 = 90, 0.5 = 180, 0.75 = 270, 1=360)", group=Detail, parameterName=Detail2Rotation}
  ) {speed=6.2832}
  n27 = Multiply(
   A=@n19
   B=ScalarParameter() {desc="Detail2 Animated Spin Speed (Pivot point isn't very predictable)", group=Detail, parameterName=Detail2Spinner}
  )
  n28 = TextureSampleParameter2D(UVs=Add(A=Multiply(A=@n12, B=@n23), B=Rotator(Coordinate=@n26, Time=@n27) {speed=6.2832}) {desc=Panner}) {group=Detail, parameterName=Detail2, samplerSource=SSM_Wrap_WorldGroupSettings, texture=&T_Black}
  n29 = LinearInterpolate(A=@n11, B=@n28.RGB, Alpha=Saturate(None=ScalarParameter() {defaultValue=0.25, group=Detail, parameterName=Detail2Strength}))
  n30 = StaticSwitchParameter(True=TextureCoordinate() {coordinateIndex=1}, False=TextureCoordinate()) {defaultValue=true, group=Decals, parameterName=*DecalsUseCylinderUV}
  n31 = VectorParameter() {desc="(R,G) = Tile (X,Y)    (B,A) = Pan (X,Y)", group=Decals, parameterName=Decal2Transform, defaultValue=<rgba 1 1 0 0>}
  n32 = ComponentMask(None=@n31.RGB) {g=true, r=true}
  n33 = Time() {bOverride_Period=true, period=20}
  n34 = VectorParameter() {desc="(R,G) = Decal 1 Pan (X,Y)    (B,A) = Decal 2 Pan (X,Y)", group=Decals, parameterName=DecalBases1*2Panner}
  n35 = Add(A=Add(A=Multiply(A=@n30, B=@n32), B=AppendVector(A=@n31.B, B=@n31.B)), B=Multiply(A=@n33, B=AppendVector(A=@n34.B, B=@n34.B))) {desc=Panner}
  n36 = ScalarParameter() {group=Decals, parameterName=Decal2Rotation}
  n37 = Time() {bOverride_Period=true, period=20}
  n38 = Rotator(
   Coordinate=StaticSwitchParameter(True=Rotator(Coordinate=@n35, Time=@n36) {speed=6.2832}, False=@n35) {desc=Enables Decal Sharpness and Threshold params., group=Decals, parameterName=Decal2Rotate}
   Time=Multiply(A=@n37, B=ScalarParameter() {desc="Decal2 Animated Spin Speed (Pivot point isn't very predictable)", group=Decals, parameterName=Decal2Spin})
  ) {speed=6.2832}
  n39 = ScalarParameter() {defaultValue=96, group=Pixelate, parameterName=NumPixels}
  n40 = TextureSampleParameter2D(UVs=StaticSwitchParameter(
    True=MaterialFunctionCall!(UV=@n38, NumPixels=@n39) {materialFunction=/VehicleCosmetics/Materials/ExtraFunctions/PixelateUV_MF}.Result
    False=@n38
   ) {group=Pixelate, parameterName=*PixelateDecal2}) {group=Decals, parameterName=Decal2, texture=&T_Black}
  n41 = StaticComponentMaskParameter(None=StaticSwitchParameter(True=OneMinus(None=@n40.RGB), False=@n40.RGB) {group=Decals, parameterName=Decal2Invert}) {defaultR=true, group=Decals, parameterName=Decal2Channel}
  n42 = Constant() {r=1}
  n43 = Add(
   A=Subtract(A=@n41, B=@n42)
   B=ScalarParameter() {defaultValue=0.5, desc=Tied to HeightMask switch., group=Decals, parameterName=Decal2Threshold}
  )
  n44 = ScalarParameter() {defaultValue=4, desc=Tied to HeightMask switch., group=Decals, parameterName=Decal2Sharpness}
  n45 = StaticSwitchParameter(True=Saturate(None=Multiply(A=@n43, B=@n44)), False=@n41) {desc=Enables Decal Sharpness and Threshold params., group=Decals, parameterName=Decal2HeightMask}
  n46 = TextureSampleParameter2D() {group=BaseTextures, parameterName=BodyTextures, samplerSource=SSM_Wrap_WorldGroupSettings, texture=&T_Black}
  n47 = StaticSwitchParameter(True=TextureCoordinate() {coordinateIndex=1, desc="UV1 (Cylinder)"}, False=TextureCoordinate() {desc="UV0 (Body Unwrap)"}) {defaultValue=true, group=Decals, parameterName=DecalBlendNoise*UseCylinderUV}
  n48 = StaticSwitchParameter(
   True=MaterialFunctionCall(UV=@n47, NumPixels=ScalarParameter() {defaultValue=96, group=Pixelate, parameterName=NumPixels}) {materialFunction=&PixelateUV_MF}.Result
   False=@n47
  ) {group=Pixelate, parameterName=*PixelateDecalBlendNoise}
  n49 = AppendVector(A=ComponentMask(None=@n48) {g=true}, B=ComponentMask(None=@n48) {r=true})
  n50 = ScalarParameter() {group=Decals, parameterName=DecalBlendRampRotation}
  n51 = StaticSwitchParameter(True=LinearInterpolate(A=@n49, B=@n48, Alpha=@n50), False=Rotator(Coordinate=@n49, Time=@n50) {speed=1}) {defaultValue=true, group=Decals, parameterName=DecalBlendRampCheapRotation}
  n52 = StaticSwitchParameter(True=@n46.R, False=ComponentMask(None=@n51) {r=true}) {desc=Instead of a linear ramp., group=Decals, parameterName=DecalBlend*UseEdgeDist}
  n53 = Constant() {r=1}
  n54 = Divide(A=Constant() {r=1}, B=ScalarParameter() {defaultValue=5, group=Decals, parameterName=DecalBlendRampWidth}) {desc=Invert multiplier}
  n55 = Divide(A=@n53, B=@n54) {desc=Get multiplier}
  n56 = Subtract(A=Constant() {r=-1}, B=@n55)
  n57 = ScalarParameter() {defaultValue=0.5, group=Decals, parameterName=DecalBlendRampPosition}
  n58 = ConstantBiasScale(None=@n57) {bias=0, scale=2}
  n59 = Add(A=@n55, B=@n53)
  n60 = Fmod(A=Multiply(A=Multiply(A=@n58, B=@n59), B=Constant() {r=0.5}), B=@n59)
  n61 = Multiply(A=Add(A=Add(A=@n52, B=@n56), B=Add(A=@n60, B=@n55)), B=@n54)
  n62 = VectorParameter() {desc="(R,G) = Tile (X,Y)    (B,A) = Pan (X,Y)", group=Decals, parameterName=DecalBlendNoiseTransform, defaultValue=<rgba 2 2 0 0>}
  n63 = AppendVector(A=@n62.B, B=@n62.B)
  n64 = StaticSwitchParameter(True=Multiply(A=@n63, B=Time() {bOverride_Period=true, period=20}), False=@n63) {desc=DecalBlendNoiseTransform vector parameter Blue and Alpha control pan speed., group=Decals, parameterName=DecalBlendNoiseDynamic}
  n65 = Add(A=Multiply(A=@n51, B=ComponentMask(None=@n62.RGB) {g=true, r=true}), B=@n64) {desc=Panner}
  n66 = Multiply(
   A=ComponentMask(None=StaticSwitchParameter(True=@n28.RGB, False=@n21.RGB) {group=Decals, parameterName=DecalBlendDistortUseDetail2}) {g=true, r=true}
   B=ScalarParameter() {defaultValue=0.25, group=Decals, parameterName=DecalBlendDistortStrength}
  )
  n67 = Rotator(
   Coordinate=StaticSwitchParameter(True=Add(A=@n65, B=@n66), False=@n65) {group=Decals, parameterName=DecalBlendDistortByDetail}
   Time=ScalarParameter() {desc="Rotate Decal Blend Noise (0.25 = 90, 0.5 = 180, 0.75 = 270, 1=360)", group=Decals, parameterName=DecalBlendNoiseRotation}
  ) {speed=6.2832}
  n68 = Multiply(
   A=Time() {bOverride_Period=true, period=20}
   B=ScalarParameter() {desc="DecalBlendNoise Animated Spin Speed (Pivot point isn't very predictable)", group=Decals, parameterName=DecalBlendNoiseSpin}
  )
  n69 = StaticComponentMaskParameter(None=TextureSampleParameter2D(UVs=Rotator(Coordinate=@n67, Time=@n68) {speed=6.2832}) {group=Decals, parameterName=DecalBlendNoise, texture=&T_Black}.RGB) {defaultR=true, desc=Uses FXNoise2, group=Decals, parameterName=DecalBlendNoiseChannel}
  n70 = OneMinus(None=@n69)
  n71 = Constant()
  n72 = Add(
   A=Subtract(
    A=StaticSwitchParameter(True=@n61, False=@n57) {defaultValue=true, desc=Instead of any ramp., group=Decals, parameterName=DecalBlend*UseRamp}
    B=Constant() {r=1}
   )
   B=StaticSwitchParameter(True=StaticSwitchParameter(True=@n70, False=@n69) {group=Decals, parameterName=DecalBlendNoiseInvert}, False=@n71) {defaultValue=true, group=Decals, parameterName=DecalBlend*UseNoise}
  )
  n73 = StaticSwitchParameter(True=@n45, False=@n71) {group=Decals, parameterName=DecalBlend*UseDecal2Height}
  n74 = Saturate(None=Multiply(A=Add(A=@n72, B=@n73), B=ScalarParameter() {defaultValue=32, group=Decals, parameterName=DecalBlendRampNoiseSharpness}))
  n75 = VectorParameter() {desc="(R,G) = Tile (X,Y)    (B,A) = Pan (X,Y)", group=Decals, parameterName=Decal1Transform, defaultValue=<rgba 1 1 0 0>}
  n76 = ComponentMask(None=@n75.RGB) {g=true, r=true}
  n77 = Add(A=Add(A=Multiply(A=@n30, B=@n76), B=AppendVector(A=@n75.B, B=@n75.B)), B=Multiply(A=@n33, B=AppendVector(A=@n34.R, B=@n34.R))) {desc=Panner}
  n78 = ScalarParameter() {group=Decals, parameterName=Decal1Rotation}
  n79 = Rotator(
   Coordinate=StaticSwitchParameter(True=Rotator(Coordinate=@n77, Time=@n78) {speed=6.2832}, False=@n77) {desc=Enables Decal Sharpness and Threshold params., group=Decals, parameterName=Decal1Rotate}
   Time=Multiply(
    A=@n37
    B=ScalarParameter() {desc="Decal 1 Animated Spin Speed (Pivot point isn't very predictable)", group=Decals, parameterName=Decal1Spin}
   )
  ) {speed=6.2832}
  n80 = TextureSampleParameter2D(UVs=StaticSwitchParameter(
    True=MaterialFunctionCall(UV=@n79, NumPixels=@n39) {materialFunction=&PixelateUV_MF}.Result
    False=@n79
   ) {group=Pixelate, parameterName=*PixelateDecal1}) {group=Decals, parameterName=Decal1, texture=&T_Black}
  n81 = StaticComponentMaskParameter(None=StaticSwitchParameter(True=OneMinus(None=@n80.RGB), False=@n80.RGB) {group=Decals, parameterName=Decal1Invert}) {defaultR=true, group=Decals, parameterName=Decal1Channel}
  n82 = Add(
   A=Subtract(A=@n81, B=@n42)
   B=ScalarParameter() {defaultValue=0.5, desc=Tied to HeightMask switch., group=Decals, parameterName=Decal1Threshold}
  )
  n83 = ScalarParameter() {defaultValue=4, desc=Tied to HeightMask switch., group=Decals, parameterName=Decal1Sharpness}
  n84 = StaticSwitchParameter(True=Saturate(None=Multiply(A=@n82, B=@n83)), False=@n81) {desc=Enables Decal Sharpness and Threshold params., group=Decals, parameterName=Decal1HeightMask}
  n85 = StaticSwitchParameter(
   True=StaticSwitchParameter(True=LinearInterpolate(A=@n84, B=@n45, Alpha=@n74), False=@n84) {defaultValue=true, group=Decals, parameterName=*DecalLayer2}
   False=Constant()
  ) {defaultValue=true, group=Decals, parameterName=*DecalLayer1}
  n86 = StaticSwitchParameter(
   True=StaticSwitchParameter(True=LinearInterpolate(A=@n22, B=@n29, Alpha=@n85), False=@n22) {group=Detail, parameterName=*Detail2}
   False=@n11
  ) {defaultValue=true, group=Detail, parameterName=*Detail}
  n87 = TextureSampleParameter2D() {
   group=BaseTextures
   parameterName=Ubershader_Normal
   samplerSource=SSM_Wrap_WorldGroupSettings
   samplerType=SAMPLERTYPE_LinearColor
   texture=/VehicleCosmetics/Textures/T_Black_Linear
  }
  n88 = LinearInterpolate(Alpha=@n87.R) {constA=-1, desc=R}
  n89 = LinearInterpolate(Alpha=@n87.G) {constA=-1, desc=G}
  n90 = Add(
   A=Multiply(A=@n86, B=Constant3Vector() {constant=<rgba 1 1 0 0>})
   B=AppendVector(A=AppendVector(A=@n88, B=@n89) {desc=R+G}, B=Constant() {desc=Original BodyN B, r=1}) {desc=R+G+B}
  )
  MaterialFunctionCall!(
   Fresnel
   Material=MakeMaterialAttributes!(
    BaseColor=LinearInterpolate(A=@n6, B=@n9.RGB, Alpha=@n10.RGB)
    Metallic=Constant() {r=0.5}
    Specular
    Roughness=Constant() {r=0.5}
    Anisotropy
    EmissiveColor
    Opacity
    OpacityMask
    Normal=StaticSwitchParameter(True=@n90, False=@n86) {group=Lighting, parameterName=*LightUseBodyNormal}
    Tangent
    WorldPositionOffset
    SubsurfaceColor
    ClearCoat
    ClearCoatRoughness
    AmbientOcclusion
    Refraction
    CustomizedUVs_0
    CustomizedUVs_1
    CustomizedUVs_2
    CustomizedUVs_3
    CustomizedUVs_4
    CustomizedUVs_5
    CustomizedUVs_6
    CustomizedUVs_7
    PixelDepthOffset
    ShadingModel
    Displacement
   )
  ) {materialFunction=/Game/Packages/Fortress_SharedMaterials/Base_Material_Functions/CharacterDeRez_Transition}
 ]
}
